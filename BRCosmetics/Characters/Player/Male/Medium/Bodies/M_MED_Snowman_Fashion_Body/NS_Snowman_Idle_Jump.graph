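# Niagara system for the Snowman body cosmetic, emitting snowflakes and cold puffs for idle, jump and landing.
# FN_BoneFlow with particle partitioning places idle and landing arm particles along the skeletal mesh, with idle puffs gated by owner speed.
# User JumpBurst and LandBurst values drive spawn probability for torus-shaped bursts and arm emitters.
# Hand/foot hidden flags set particle visibility tags, and distance-from-owner curves fade arm particles.
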NiagaraSystem /BRCosmetics/Characters/Player/Male/Medium/Bodies/M_MED_Snowman_Fashion_Body/NS_Snowman_Idle_Jump
dependencies {
 usedDataInterfaces [&NiagaraDataInterfaceColorCurve, &NiagaraDataInterfaceCurve]
 usedDynamicInputs [&Boolean_AndOperation]
 usedModules [
  &SystemState
  &FN_Integrator
  &self:SystemSpawnScript.SystemScriptSource.SystemScriptGraph.NiagaraNodeAssignment_0.SetVariables_#1
  &self:SystemSpawnScript.SystemScriptSource.SystemScriptGraph.NiagaraNodeAssignment_1.SetVariables_#2
 ]
}
emitters {
 Idle_Flakes {
  emitterSpawnScript [
   /SetVariables(Emitter.RandomSeed:NiagaraInt32=/Niagara/DynamicInputs/Random/RandomInt(Randomness Mode:ENiagaraRandomnessMode=::NewEnumerator0, Random Seed:NiagaraInt32=0, Fixed Random Seed:NiagaraBool=false)) {isSetParameters=true, module=SetVariables_#3}
   /Game/Effects/Niagara/Modules/Partition/FN_PartitionAttributes(Number of Bone Sets=2, Particles Per Partition=8, Partitions=3)
  ]
  emitterUpdateScript [
   &EmitterState(Life Cycle Mode=::NewEnumerator1, Loop Duration=1, UseLoopDelay=false)
   &SpawnBurst_Instantaneous(Spawn Count=$Emitter.FN_PartitionAttributes.Spawn Total, Use Spawn Probability=false) {enabled=false}
   &SpawnRate(SpawnRate=10, Use Spawn Probability=true)
   &SpawnRate(
    SpawnRate=&MakeCustomFloatFromBool(
     Bool=&SetBoolByFloatComparison(
      A=&VectorLength(Vector=&NormalizeVector(Vector To Normalize=$Engine.Owner.Velocity))
      B=0.01
     )
     True Float=0
     False Float=30
    )
    Use Spawn Probability=true
   ) {enabled=false, module=SpawnRate001}
   &SpawnPerUnit(Module.Use Max Movement Threshold=true, Module.Use Movement Tolerance=true, Use Spawn Probability=false) {enabled=false}
  ]
  particleSpawnScript [
   /SetVariables(
    Particles.BoneFlow Height=1
    Particles.Boneflow Radius:NiagaraFloat=&Lerp_Float(A=8, B=5, Alpha=$Particles.BoneFlow Height)
    Particles.RandomRangeAngle:NiagaraFloat=&RandomRangeFloat(Minimum=-0.3, Maximum=0.3)
   ) {isSetParameters=true, module=SetVariables_#4}
   /Game/Effects/Niagara/Modules/Partition/FN_Particle_partition_v5/FN_Particle_Partition_v5(Pattern:ENiagaraPartitionWaveForm=::NewEnumerator1)
   /Niagara/Modules/Spawn/Initialization/InitializeParticle(
    Lifetime=&RandomRangeFloat(Minimum=1.5, Maximum=2)
    Module.Write Color=true
    Module.Write Lifetime=true
    Module.Write Mass=true
    Module.Write Material Random=true
    Module.Write Position=false
    Module.Write Scale=false
    Module.Write SpriteRotation=true
    Module.Write SpriteSize=true
    Position=&SimulationPosition()
    Sprite Rotation=&RandomRangeFloat(Minimum=0, Maximum=1)
    Sprite Size=&Vector2DFromFloat(Value=&RandomRangeFloat(Minimum=2, Maximum=5))
   )
   &Color(Color=<rgba 0.4 0.59 1 1>, Scale Alpha=$Particles.MaterialRandom)
   &FN_BoneFlow(
    Angle:NiagaraFloat=&Lerp_Float(
     A=$Particles.RandomRangeAngle
     B=&Add_Float(A=$Particles.RandomRangeAngle, B=0.5)
     Alpha=&ModuloFloat(A=&MakeFloatFromInt(Int=$Particles.Partition.ID), B=2)
    )
    Height:NiagaraFloat=&Max_Float(A:NiagaraFloat=$Particles.MaterialRandom, B:NiagaraFloat=0.001)
    Interpolate Mesh Sampling=false
    Radius:NiagaraFloat=$Particles.Boneflow Radius
    SkeletalMesh:NiagaraDataInterfaceSkeletalMesh=$User.SkeletalMesh
    Write Position Attribute=true
   )
   &InheritVelocity(Module.Limit Speed=true, Velocity Scale=&VectorFromFloat(Value=5))
   &AddVelocity(Velocity=&Multiply_VectorByFloat(Vector=$Output.FN_BoneFlow.SplineTangent, Float=-5))
   &AddVelocity(Velocity=&Multiply_VectorByFloat(Vector=$Output.FN_BoneFlow.RotatedBiNormal, Float=5)) {enabled=false, module=AddVelocity001}
  ]
  particleUpdateScript [
   /SetVariables(Particles.Boneflow Radius:NiagaraFloat=&Multiply_Float(
     A=$Particles.Boneflow Radius
     B=&FloatFromCurve(FloatCurve=DataInterface() {propertyValues {Curve {keys [<at 0 1>,<at 1 15>]}}})
    )) {enabled=false, isSetParameters=true, module=SetVariables_#5}
   &FN_BoneFlow(
    Angle:NiagaraFloat=&Lerp_Float(
     A=$Particles.RandomRangeAngle
     B=&Add_Float(A=$Particles.RandomRangeAngle, B=0.5)
     Alpha=&ModuloFloat(A=&MakeFloatFromInt(Int=$Particles.Partition.ID), B=2)
    )
    Height:NiagaraFloat=&Max_Float(A:NiagaraFloat=$Particles.MaterialRandom, B:NiagaraFloat=0.001)
    Interpolate Mesh Sampling=false
    Radius:NiagaraFloat=&Multiply_Float(
     A=$Particles.Boneflow Radius
     B=&FloatFromCurve(FloatCurve=DataInterface() {propertyValues {Curve {keys [<at 0 1>,<at 1 2>]}}}, Scale Curve=5)
    )
    SkeletalMesh:NiagaraDataInterfaceSkeletalMesh=$User.SkeletalMesh
    Write Position Attribute=true
   ) {enabled=false, module=FN_BoneFlow001}
   &AccelerationForce(Acceleration=<0 0 -50>)
   &Drag(
    Drag=&RandomRangeFloat(Minimum=2, Maximum=5)
    Rotational Drag=20
    Use Linear Drag=true
    Use Rotational Drag=false
   )
   &SolveForcesAndVelocity(Clamp Velocity=false, Limit Acceleration=false)
   &ParticleState()
   &SpriteRotationRate(Rotation Rate=&FloatFromCurve(
     FloatCurve=DataInterface() {propertyValues {Curve {keys [<at 0 0 RCIM_Cubic>,<at 0.3 1 RCIM_Cubic>,<at 1 0>]}}}
     Scale Curve=&RandomRangeFloat(Minimum=-100, Maximum=100)
    ))
   &ScaleColor(
    Scale Alpha=&FloatFromCurve(FloatCurve=DataInterface() {propertyValues {Curve {keys [<at 0.6 1>,<at 1 0>]}}})
    ScaleA=true
    ScaleRGB=true
   )
   &ScaleSpriteSize(Scale Factor=&Vector2DFromFloat(Value=&FloatFromCurve(FloatCurve=DataInterface() {propertyValues {Curve {keys [
          <at 0 0 RCIM_Cubic>
          {arriveTangent=2, interpMode=RCIM_Cubic, leaveTangent=2, time=0.1619, value=1}
          {arriveTangent=-1.1932, interpMode=RCIM_Cubic, leaveTangent=-1.1932, time=0.5, value=1}
          <at 1 0 RCIM_Cubic>
         ]}}})))
   &ScaleColor(
    Scale Alpha=&FloatFromCurve(
     FloatCurve=$System.ScaleDistanceFadeCurve
     CurveIndex=&Multiply_Float(
      A=&VectorLength(Vector=&Subtract_Vector(
        A=Linked() {linkedVariable {name=Particles.Position, type=NiagaraPosition}}
        B=&Add_Vector(A=Linked() {linkedVariable {name=Engine.Owner.Position, type=NiagaraPosition}}, B=$System.ScaleDistanceOffset)
       ))
      B=0.003
     )
    )
    ScaleA=true
    ScaleRGB=true
   ) {module=ScaleColor002}
   /SetVariables(Particles.VisibilityTag=&MakeCustomIntFromBool(Bool=$User.bisHandHidden, True INT=1, False INT=0)) {isSetParameters=true, module=SetVariables_#6}
  ]
  rendererClasses [&NiagaraSpriteRendererProperties]
  renderers [{rendererClass=&NiagaraSpriteRendererProperties}]
 }
 Idle_Puffs {
  emitterSpawnScript [
   /SetVariables(Emitter.RandomSeed:NiagaraInt32=&RandomInt(Randomness Mode:ENiagaraRandomnessMode=::NewEnumerator0, Random Seed:NiagaraInt32=0, Fixed Random Seed:NiagaraBool=false)) {isSetParameters=true, module=SetVariables_#3}
   &FN_PartitionAttributes(Number of Bone Sets=2, Particles Per Partition=8, Partitions=3)
  ]
  emitterUpdateScript [
   &EmitterState(Life Cycle Mode=::NewEnumerator1, Loop Duration=1, UseLoopDelay=false)
   &SpawnRate(
    Spawn Probability=&MakeCustomFloatFromBool(
     Bool=&SetBoolByFloatComparison(A=&VectorLength(Vector=$Engine.Owner.Velocity), B=1)
     True Float=0
     False Float=1
    )
    SpawnRate=10
    Use Spawn Probability=true
   ) {module=SpawnRate002}
  ]
  particleSpawnScript [
   /SetVariables(
    Particles.BoneFlow Height=1
    Particles.Boneflow Radius:NiagaraFloat=&RandomRangeFloat(Minimum=8, Maximum=5)
    Particles.RandomRangeAngle:NiagaraFloat=&RandomRangeFloat(Minimum=-0.3, Maximum=0.3)
   ) {isSetParameters=true, module=SetVariables_#4}
   &FN_Particle_Partition_v5(Pattern:ENiagaraPartitionWaveForm=::NewEnumerator1)
   &InitializeParticle(
    Color=&ColorFromCurve(
     ColorCurve=DataInterface() {propertyValues {
       AlphaCurve {keys [<at 0 0.1 RCIM_Cubic>,<at 1 0.05 RCIM_Cubic>]}
       BlueCurve {keys [<at 0 1 RCIM_Cubic>,<at 1 1 RCIM_Cubic>]}
       GreenCurve {keys [<at 0 1 RCIM_Cubic>,<at 1 0.8149 RCIM_Cubic>]}
       RedCurve {keys [<at 0 1 RCIM_Cubic>,<at 1 0.8 RCIM_Cubic>]}
      }}
     CurveIndex=$Particles.Partition.NormalizedID
    )
    Lifetime=&RandomRangeFloat(Minimum=1.4, Maximum=2)
    Module.Write Color=true
    Module.Write Lifetime=true
    Module.Write Mass=true
    Module.Write Material Random=true
    Module.Write Position=false
    Module.Write Scale=false
    Module.Write SpriteRotation=true
    Module.Write SpriteSize=true
    Position=&SimulationPosition()
    Sprite Rotation=&RandomRangeFloat(Minimum=0, Maximum=360)
    Sprite Size=&Vector2DFromFloat(Value=&RandomRangeFloat(Minimum=10, Maximum=10))
   )
   &Color(Color=<rgba 0.4 0.59 1 1>, Scale Alpha=$Particles.MaterialRandom) {enabled=false}
   &FN_BoneFlow(
    Angle:NiagaraFloat=&Lerp_Float(
     A=$Particles.RandomRangeAngle
     B=&Add_Float(A=$Particles.RandomRangeAngle, B=0.5)
     Alpha=&ModuloFloat(A=&MakeFloatFromInt(Int=$Particles.Partition.ID), B=2)
    )
    Height:NiagaraFloat=&Max_Float(A:NiagaraFloat=$Particles.MaterialRandom, B:NiagaraFloat=0.001)
    Interpolate Mesh Sampling=false
    Radius:NiagaraFloat=$Particles.Boneflow Radius
    SkeletalMesh:NiagaraDataInterfaceSkeletalMesh=$User.SkeletalMesh
    Write Position Attribute=true
   )
   &AddVelocity(Velocity=&Multiply_VectorByFloat(Vector=$Output.FN_BoneFlow.SplineTangent, Float=-5))
   &AddVelocity(Velocity=&Multiply_VectorByFloat(Vector=$Output.FN_BoneFlow.RotatedBiNormal, Float=5)) {enabled=false, module=AddVelocity001}
  ]
  particleUpdateScript [
   &AccelerationForce(Acceleration=&RandomRangeVector(Minimum=<0 0 0>, Maximum=<1 1 -100>))
   &InheritVelocity(Module.Limit Speed=true, Velocity Limit=200, Velocity Scale=&VectorFromFloat(Value=-1)) {enabled=false}
   &Drag(
    Drag=&RandomRangeFloat(Minimum=5, Maximum=5)
    Rotational Drag=20
    Use Linear Drag=true
    Use Rotational Drag=false
   )
   &SolveForcesAndVelocity(Clamp Velocity=false, Limit Acceleration=false)
   &ParticleState()
   /Niagara/Modules/Update/SubUV/SubUVAnimation(Number Of Frames=&Lerp_Float(A=50, B=50, Alpha=$Particles.NormalizedAge)) {enabled=false}
   /Niagara/Modules/Update/SubUV/V2/SubUVAnimation(
    Sub UV Play Rate:NiagaraFloat=&Multiply_Float(A=0.3, B=$Particles.NormalizedAge)
    SubUV Animation Mode:ENiagara_SubUVLookupMode=::NewEnumerator3
   ) {module=SubUVAnimation001}
   &SpriteRotationRate(Rotation Rate=&FloatFromCurve(
     FloatCurve=DataInterface() {propertyValues {Curve {keys [<at 0 0 RCIM_Cubic>,<at 0.3 1 RCIM_Cubic>,<at 1 0>]}}}
     Scale Curve=&RandomRangeFloat(Minimum=-20, Maximum=20)
    ))
   &ScaleSpriteSize(Scale Factor=&Vector2DFromFloat(Value=&FloatFromCurve(
      FloatCurve=DataInterface() {propertyValues {Curve {keys [
          {arriveTangent=0.7564, interpMode=RCIM_Cubic, leaveTangent=0.7564, tangentMode=RCTM_User, time=-0.0024, value=0.2}
          {arriveTangent=1.4669, interpMode=RCIM_Cubic, leaveTangent=1.4669, time=0.4362, value=1.0444}
          <at 0.9951 1.6633 RCIM_Cubic>
         ]}}}
      Scale Curve=2
     )))
   &ScaleColor(
    Scale Mode=::NewEnumerator1
    Scale RGBA=&MakeVector4_FromLinearColor(LinearColor=&ColorFromCurve(ColorCurve=DataInterface() {propertyValues {
        AlphaCurve {keys [
          <at 0 0 RCIM_Cubic>
          {arriveTangent=1.5193, interpMode=RCIM_Cubic, leaveTangent=1.5193, time=0.2235, value=0.9746}
          {arriveTangent=-1.2368, interpMode=RCIM_Cubic, leaveTangent=-1.2368, time=0.5352, value=0.8131}
          <at 1.0115 0>
          <at 1.094 0>
         ]}
        BlueCurve {keys [<at 0 1 RCIM_Cubic>,<at 1 1 RCIM_Cubic>]}
        GreenCurve {keys [<at 0 0.715 RCIM_Cubic>,<at 1 0.79 RCIM_Cubic>]}
        RedCurve {keys [<at 0 0.62 RCIM_Cubic>,<at 1 0.7 RCIM_Cubic>]}
       }}))
    ScaleRGBA=true
   ) {module=ScaleColor001}
   /SetVariables(Particles.Lifetime=&Lerp_Float(
     A=$Particles.Initial.Lifetime
     B=2
     Alpha=&VectorLength(Vector=&Multiply_VectorByFloat(Vector=$Engine.Owner.Velocity, Float=1))
    )) {enabled=false, isSetParameters=true, module=SetVariables_#7}
   /Niagara/Modules/Update/Color/ScaleColorBySpeed(A Scale Min=0.2, Max Speed Threshold=150, Min Speed Threshold=10, RGB Scale Min=1, Source Velocity:Vector3f=$Engine.Owner.Velocity) {enabled=false}
   /SetVariables(
    Particles.Alpha:NiagaraFloat=&MakeFloatFromVector4(Vector 4=&MakeVector4_FromLinearColor(LinearColor=$Particles.Color), Channel=::NewEnumerator3)
    Particles.Color RGB:Vector3f=&MakeVectorFromLinearColorRGB(Color=$Particles.Color)
   ) {enabled=false, isSetParameters=true, module=SetVariables_#8}
   /SetVariables(Particles.Color=&MakeLinearColorFromVectorAndFloat(
     "Vector (RGB)"=$Particles.Color RGB
     "Float (Alpha)"=&MakeCustomFloatFromBool(
      Bool=&SetBoolByFloatComparison(A=$Particles.Alpha, B=0.2, Comparison Type=::NewEnumerator6)
      True Float=0
      False Float=1
     )
    )) {enabled=false, isSetParameters=true, module=SetVariables_#9}
   &ScaleColor(
    Scale Alpha=&Lerp_Float(
     A=$System.ScaleDistanceAlphaMin
     B=$System.ScaleDistanceAlphaMax
     Alpha=&Multiply_Float(
      A=&VectorLength(Vector=&Subtract_Vector(
        A=Linked() {linkedVariable {name=Particles.Position, type=NiagaraPosition}}
        B=&Add_Vector(A=Linked() {linkedVariable {name=Engine.Owner.Position, type=NiagaraPosition}}, B=$System.ScaleDistanceOffset)
       ))
      B=$System.ScaleDistanceBias
     )
    )
    Scale RGB=&Lerp_Vector(
     A=$System.ScaleDistanceColorMin
     B=$System.ScaleDistanceColorMax
     Alpha=&Lerp_Float(
      A=$System.ScaleDistanceAlphaMin
      B=$System.ScaleDistanceAlphaMax
      Alpha=&Multiply_Float(
       A=&VectorLength(Vector=&Subtract_Vector(
         A=Linked() {linkedVariable {name=Particles.Position, type=NiagaraPosition}}
         B=&Add_Vector(A=Linked() {linkedVariable {name=Engine.Owner.Position, type=NiagaraPosition}}, B=$System.ScaleDistanceOffset)
        ))
       B=$System.ScaleDistanceBias
      )
     )
    )
    ScaleA=true
    ScaleRGB=true
   ) {enabled=false, module=ScaleColor002}
   &ScaleColor(
    Scale Alpha=&FloatFromCurve(
     FloatCurve=$System.ScaleDistanceFadeCurve
     CurveIndex=&Multiply_Float(
      A=&VectorLength(Vector=&Subtract_Vector(
        A=Linked() {linkedVariable {name=Particles.Position, type=NiagaraPosition}}
        B=&Add_Vector(A=Linked() {linkedVariable {name=Engine.Owner.Position, type=NiagaraPosition}}, B=$System.ScaleDistanceOffset)
       ))
      B=$System.ScaleDistanceBias
     )
    )
    ScaleA=true
    ScaleRGB=true
   )
   /SetVariables(Particles.VisibilityTag=&MakeCustomIntFromBool(Bool=$User.bisHandHidden, True INT=1, False INT=0)) {isSetParameters=true, module=SetVariables_#6}
  ]
  rendererClasses [&NiagaraSpriteRendererProperties]
  renderers [{rendererClass=&NiagaraSpriteRendererProperties}]
 }
 JumpBurst001 {
  emitterUpdateScript [
   &EmitterState(Life Cycle Mode=::NewEnumerator1, Loop Duration=1, UseLoopDelay=false)
   &SpawnRate(Spawn Probability=$User.JumpBurst, SpawnRate=500, Use Spawn Probability=true)
  ]
  particleSpawnScript [
   &InitializeParticle(
    Color=<rgba 1 1 1 0.1>
    Lifetime=&RandomRangeFloat(Minimum=0.7, Maximum=1.4)
    Module.Write Color=true
    Module.Write Lifetime=true
    Module.Write Mass=true
    Module.Write Material Random=true
    Module.Write Position=false
    Module.Write Scale=false
    Module.Write SpriteRotation=true
    Module.Write SpriteSize=true
    Position=&SimulationPosition()
    Sprite Rotation=&RandomRangeFloat(Minimum=-90, Maximum=90)
    Sprite Size=&Vector2DFromFloat(Value=&RandomRangeFloat(Minimum=15, Maximum=35))
   )
   &TorusLocation(
    Align Torus With Axis=false
    Handle Radius=0
    Large Radius=20
    MaskSpawnGroup=false
    Module.Override Randomness=false
    Torus Origin:NiagaraPosition=&Add_Vector(A=Linked() {linkedVariable {name=Particles.Position, type=NiagaraPosition}}, B=<0 0 -20>)
   )
   &AddVelocityFromPoint(Module.Use Falloff Distance=false, Velocity Strength=20)
   &AddVelocity(Velocity=<0 0 150>)
  ]
  particleUpdateScript [
   &ParticleState()
   &AccelerationForce(Acceleration=<0 0 -300>)
   &Drag(Drag=&RandomRangeFloat(Minimum=2, Maximum=3), Use Linear Drag=true, Use Rotational Drag=false) {enabled=false}
   &SolveForcesAndVelocity(Clamp Velocity=false, Limit Acceleration=false)
   &SpriteRotationRate(Rotation Rate=&FloatFromCurve(
     FloatCurve=DataInterface() {propertyValues {Curve {keys [<at 0 0 RCIM_Cubic>,<at 0.3 1 RCIM_Cubic>,<at 1 0>]}}}
     Scale Curve=&RandomRangeFloat(Minimum=-100, Maximum=100)
    ))
   &SubUVAnimation(
    Sub UV Play Rate:NiagaraFloat=&Multiply_Float(A=&Multiply_Float(A=$Particles.MaterialRandom, B=0.6), B=$Particles.NormalizedAge)
    SubUV Animation Mode:ENiagara_SubUVLookupMode=::NewEnumerator3
    SubUV Start Time Offset=1
   ) {module=SubUVAnimation001}
   &ScaleColor(
    Scale Mode=::NewEnumerator1
    Scale RGBA=&MakeVector4_FromLinearColor(LinearColor=&ColorFromCurve(ColorCurve=DataInterface() {propertyValues {
        AlphaCurve {keys [<at 0 0 RCIM_Cubic>,<at 0.04 1>,<at 0.135 0.62>,<at 1.0115 0>,<at 1.094 0>]}
        BlueCurve {keys [<at 0 1 RCIM_Cubic>,<at 1 1 RCIM_Cubic>]}
        GreenCurve {keys [<at 0 0.715 RCIM_Cubic>,<at 1 0.79 RCIM_Cubic>]}
        RedCurve {keys [<at 0 0.62 RCIM_Cubic>,<at 1 0.7 RCIM_Cubic>]}
       }}))
    ScaleRGBA=true
   ) {module=ScaleColor001}
   &ScaleSpriteSize(Scale Factor=&Vector2DFromFloat(Value=&FloatFromCurve(
      FloatCurve=DataInterface() {propertyValues {Curve {keys [
          {arriveTangent=3.9723, interpMode=RCIM_Cubic, leaveTangent=3.9723, tangentMode=RCTM_User, time=0, value=0}
          {arriveTangent=1, interpMode=RCIM_Cubic, leaveTangent=1, time=0.209, value=0.58}
          <at 1 1 RCIM_Cubic>
         ]}}}
      Scale Curve=3
     )))
   /SetVariables(Particles.VisibilityTag=&MakeCustomIntFromBool(Bool=$User.bisFootHidden, True INT=1, False INT=0)) {isSetParameters=true, module=SetVariables_#10}
  ]
  rendererClasses [&NiagaraSpriteRendererProperties]
  renderers [{rendererClass=&NiagaraSpriteRendererProperties}]
 }
 JumpSpawnRate001 {
  emitterUpdateScript [
   &EmitterState(Life Cycle Mode=::NewEnumerator1, Loop Duration=1, UseLoopDelay=false)
   &SpawnRate(
    SpawnRate=&Multiply_Float(
     A=100
     B=&Multiply_Float(
      A=$System.FWD Imitation Burst
      B=&OneMinusFloat(Float=&Multiply_Float(A=$System.FN_Integrator.Value, B=2))
     )
    )
    Use Spawn Probability=true
   )
  ]
  particleSpawnScript [
   &InitializeParticle(
    Lifetime=0.5
    Module.Write Color=true
    Module.Write Lifetime=true
    Module.Write Mass=true
    Module.Write Material Random=true
    Module.Write Position=false
    Module.Write Scale=false
    Module.Write SpriteRotation=false
    Module.Write SpriteSize=true
    Position=&SimulationPosition()
    Sprite Size=&Vector2DFromFloat(Value=&RandomRangeFloat(Minimum=3, Maximum=5))
   )
   &Color(Color=<rgba 0.4 0.59 1 0.3>, Scale Alpha=$Particles.MaterialRandom)
   &TorusLocation(Align Torus With Axis=false, Handle Radius=0, Large Radius=10, MaskSpawnGroup=false, Module.Override Randomness=false)
   &AddVelocityFromPoint(Module.Use Falloff Distance=false, Velocity Strength=10)
  ]
  particleUpdateScript [
   &ParticleState()
   &Drag(Drag=5, Use Linear Drag=true, Use Rotational Drag=false)
   &SolveForcesAndVelocity(Clamp Velocity=false, Limit Acceleration=false)
   &ScaleColor(
    Scale RGB=&MakeVectorFromLinearColorRGB(Color=&ColorFromCurve(ColorCurve=DataInterface() {propertyValues {
        AlphaCurve {keys [<at 0 0 RCIM_Cubic>,<at 0.04 1>,<at 0.135 0.62>,<at 1.0115 0>,<at 1.094 0>]}
        BlueCurve {keys [<at 0 1 RCIM_Cubic>,<at 1 1 RCIM_Cubic>]}
        GreenCurve {keys [<at 0 0.8875 RCIM_Cubic>,<at 1 0.79 RCIM_Cubic>]}
        RedCurve {keys [<at 0 0.85 RCIM_Cubic>,<at 1 0.7 RCIM_Cubic>]}
       }}))
    ScaleA=true
    ScaleRGB=true
   )
   &ScaleSpriteSize(Scale Factor=&Vector2DFromFloat(Value=&FloatFromCurve(FloatCurve=DataInterface() {propertyValues {Curve {keys [
          <at 0 0 RCIM_Cubic>
          {arriveTangent=2, interpMode=RCIM_Cubic, leaveTangent=2, time=0.1619, value=1}
          {arriveTangent=-1.1932, interpMode=RCIM_Cubic, leaveTangent=-1.1932, time=0.5, value=1}
          <at 1 0 RCIM_Cubic>
         ]}}})))
   /SetVariables(Particles.VisibilityTag=&MakeCustomIntFromBool(Bool=$User.bisFootHidden, True INT=1, False INT=0)) {isSetParameters=true, module=SetVariables_#10}
  ]
  rendererClasses [&NiagaraSpriteRendererProperties]
  renderers [{rendererClass=&NiagaraSpriteRendererProperties}]
 }
 Land_ArmsSnowPuffs {
  emitterSpawnScript [
   /SetVariables(Emitter.RandomSeed:NiagaraInt32=&RandomInt(Randomness Mode:ENiagaraRandomnessMode=::NewEnumerator0, Random Seed:NiagaraInt32=0, Fixed Random Seed:NiagaraBool=false)) {isSetParameters=true, module=SetVariables_#3}
   &FN_PartitionAttributes(Number of Bone Sets=2, Particles Per Partition=8, Partitions=3)
  ]
  emitterUpdateScript [
   &EmitterState(Life Cycle Mode=::NewEnumerator1, Loop Duration=1, UseLoopDelay=false)
   &SpawnRate(Spawn Probability=$User.LandBurst, SpawnRate=200, Use Spawn Probability=true)
  ]
  particleSpawnScript [
   /SetVariables(
    Particles.BoneFlow Height=0.1
    Particles.Boneflow Radius:NiagaraFloat=&RandomRangeFloat(Minimum=1, Maximum=2)
    Particles.RandomRangeAngle:NiagaraFloat=&RandomRangeFloat(Minimum=0, Maximum=0)
   ) {isSetParameters=true, module=SetVariables_#4}
   &FN_Particle_Partition_v5(Pattern:ENiagaraPartitionWaveForm=::NewEnumerator1)
   &InitializeParticle(
    Color=&ColorFromCurve(
     ColorCurve=DataInterface() {propertyValues {
       AlphaCurve {keys [<at 0 0.1 RCIM_Cubic>,<at 1 0.05 RCIM_Cubic>]}
       BlueCurve {keys [<at 0 1 RCIM_Cubic>,<at 1 1 RCIM_Cubic>]}
       GreenCurve {keys [<at 0 1 RCIM_Cubic>,<at 1 0.8149 RCIM_Cubic>]}
       RedCurve {keys [<at 0 1 RCIM_Cubic>,<at 1 0.8 RCIM_Cubic>]}
      }}
     CurveIndex=$Particles.Partition.NormalizedID
    )
    Lifetime=&RandomRangeFloat(Minimum=0.8, Maximum=1.2)
    Module.Write Color=true
    Module.Write Lifetime=true
    Module.Write Mass=true
    Module.Write Material Random=true
    Module.Write Position=false
    Module.Write Scale=false
    Module.Write SpriteRotation=true
    Module.Write SpriteSize=true
    Position=&SimulationPosition()
    Sprite Rotation=&RandomRangeFloat(Minimum=0, Maximum=360)
    Sprite Size=&Vector2DFromFloat(Value=&RandomRangeFloat(Minimum=8, Maximum=10))
   )
   &Color(Color=<rgba 0.4 0.59 1 1>, Scale Alpha=$Particles.MaterialRandom) {enabled=false}
   &FN_BoneFlow(
    Angle:NiagaraFloat=&Lerp_Float(
     A=$Particles.RandomRangeAngle
     B=&Add_Float(A=$Particles.RandomRangeAngle, B=0.5)
     Alpha=&ModuloFloat(A=&MakeFloatFromInt(Int=$Particles.Partition.ID), B=2)
    )
    Height:NiagaraFloat=&Max_Float(A:NiagaraFloat=$Particles.MaterialRandom, B:NiagaraFloat=0.001)
    Interpolate Mesh Sampling=false
    Radius:NiagaraFloat=$Particles.Boneflow Radius
    SkeletalMesh:NiagaraDataInterfaceSkeletalMesh=$User.SkeletalMesh
    Write Position Attribute=true
   )
   &AddVelocity(Velocity=&Multiply_VectorByFloat(Vector=$Output.FN_BoneFlow.SplineTangent, Float=10))
   &AddVelocityFromPoint(Module.Use Falloff Distance=false, Origin Offset Coordinate Space=::World, Velocity Strength=50) {enabled=false}
   &AddVelocity(Velocity=<0 0 20>) {module=AddVelocity001}
  ]
  particleUpdateScript [
   &Drag(
    Drag=&RandomRangeFloat(Minimum=5, Maximum=5)
    Rotational Drag=20
    Use Linear Drag=true
    Use Rotational Drag=false
   ) {enabled=false}
   &AccelerationForce(Acceleration=&RandomRangeVector(Minimum=<0 0 -50>, Maximum=<0 0 -100>))
   &SolveForcesAndVelocity(Clamp Velocity=false, Limit Acceleration=false)
   &ParticleState()
   /Niagara/Modules/Update/SubUV/SubUVAnimation(Number Of Frames=&Lerp_Float(A=50, B=50, Alpha=$Particles.NormalizedAge)) {enabled=false}
   /Niagara/Modules/Update/SubUV/V2/SubUVAnimation(
    Sub UV Play Rate:NiagaraFloat=&Multiply_Float(A=0.5, B=$Particles.NormalizedAge)
    SubUV Animation Mode:ENiagara_SubUVLookupMode=::NewEnumerator3
   ) {module=SubUVAnimation001}
   &SpriteRotationRate(Rotation Rate=&FloatFromCurve(
     FloatCurve=DataInterface() {propertyValues {Curve {keys [<at 0 0 RCIM_Cubic>,<at 0.3 1 RCIM_Cubic>,<at 1 0>]}}}
     Scale Curve=&RandomRangeFloat(Minimum=-20, Maximum=20)
    ))
   &ScaleSpriteSize(Scale Factor=&Vector2DFromFloat(Value=&FloatFromCurve(
      FloatCurve=DataInterface() {propertyValues {Curve {keys [
          {arriveTangent=0.7564, interpMode=RCIM_Cubic, leaveTangent=0.7564, tangentMode=RCTM_User, time=-0.0024, value=0.2}
          {arriveTangent=1.4669, interpMode=RCIM_Cubic, leaveTangent=1.4669, time=0.3797, value=1.3341}
          <at 0.9951 1.6633 RCIM_Cubic>
         ]}}}
      Scale Curve=3
     )))
   &ScaleColor(
    Scale Mode=::NewEnumerator1
    Scale RGBA=&MakeVector4_FromLinearColor(LinearColor=&ColorFromCurve(ColorCurve=DataInterface() {propertyValues {
        AlphaCurve {keys [
          <at 0 0 RCIM_Cubic>
          {arriveTangent=1.5193, interpMode=RCIM_Cubic, leaveTangent=1.5193, time=0.2235, value=0.9746}
          {arriveTangent=-1.2368, interpMode=RCIM_Cubic, leaveTangent=-1.2368, time=0.5352, value=0.8131}
          <at 1.0115 0>
          <at 1.094 0>
         ]}
        BlueCurve {keys [<at 0 1 RCIM_Cubic>,<at 1 1 RCIM_Cubic>]}
        GreenCurve {keys [<at 0 0.715 RCIM_Cubic>,<at 1 0.79 RCIM_Cubic>]}
        RedCurve {keys [<at 0 0.62 RCIM_Cubic>,<at 1 0.7 RCIM_Cubic>]}
       }}))
    ScaleRGBA=true
   ) {module=ScaleColor001}
   /SetVariables(Particles.Lifetime=&Lerp_Float(
     A=$Particles.Initial.Lifetime
     B=2
     Alpha=&VectorLength(Vector=&Multiply_VectorByFloat(Vector=$Engine.Owner.Velocity, Float=1))
    )) {enabled=false, isSetParameters=true, module=SetVariables_#7}
   &ScaleColorBySpeed(A Scale Min=0.2, Max Speed Threshold=150, Min Speed Threshold=10, RGB Scale Min=1, Source Velocity:Vector3f=$Engine.Owner.Velocity) {enabled=false}
   /SetVariables(
    Particles.Alpha:NiagaraFloat=&MakeFloatFromVector4(Vector 4=&MakeVector4_FromLinearColor(LinearColor=$Particles.Color), Channel=::NewEnumerator3)
    Particles.Color RGB:Vector3f=&MakeVectorFromLinearColorRGB(Color=$Particles.Color)
   ) {enabled=false, isSetParameters=true, module=SetVariables_#8}
   /SetVariables(Particles.Color=&MakeLinearColorFromVectorAndFloat(
     "Vector (RGB)"=$Particles.Color RGB
     "Float (Alpha)"=&MakeCustomFloatFromBool(
      Bool=&SetBoolByFloatComparison(A=$Particles.Alpha, B=0.2, Comparison Type=::NewEnumerator6)
      True Float=0
      False Float=1
     )
    )) {enabled=false, isSetParameters=true, module=SetVariables_#9}
   &ScaleColor(
    Scale RGB=&Lerp_Vector(
     A=<1 1 1>
     B=<200 1 1>
     Alpha=&Lerp_Float(
      A=1
      B=0
      Alpha=&VectorLength(Vector=&Subtract_Vector(
        A=Linked() {linkedVariable {name=Particles.Position, type=NiagaraPosition}}
        B=Linked() {linkedVariable {name=Engine.Owner.Position, type=NiagaraPosition}}
       ))
     )
    )
    ScaleA=true
    ScaleRGB=true
   ) {enabled=false}
   &ScaleColor(
    Scale Alpha=&FloatFromCurve(
     FloatCurve=$System.ScaleDistanceFadeCurve
     CurveIndex=&Multiply_Float(
      A=&VectorLength(Vector=&Subtract_Vector(
        A=Linked() {linkedVariable {name=Particles.Position, type=NiagaraPosition}}
        B=&Add_Vector(A=Linked() {linkedVariable {name=Engine.Owner.Position, type=NiagaraPosition}}, B=<0 0 50>)
       ))
      B=$System.ScaleDistanceBias
     )
     Scale Curve=5
    )
    ScaleA=true
    ScaleRGB=true
   ) {module=ScaleColor002}
   /SetVariables(Particles.VisibilityTag=&MakeCustomIntFromBool(Bool=$User.bisHandHidden, True INT=1, False INT=0)) {isSetParameters=true, module=SetVariables_#11}
  ]
  rendererClasses [&NiagaraSpriteRendererProperties]
  renderers [{rendererClass=&NiagaraSpriteRendererProperties}]
 }
 Land_ArmsSnowflakes {
  emitterSpawnScript [
   /SetVariables(Emitter.RandomSeed:NiagaraInt32=&RandomInt(Randomness Mode:ENiagaraRandomnessMode=::NewEnumerator0, Random Seed:NiagaraInt32=0, Fixed Random Seed:NiagaraBool=false)) {isSetParameters=true, module=SetVariables_#3}
   &FN_PartitionAttributes(Number of Bone Sets=2, Particles Per Partition=8, Partitions=3)
  ]
  emitterUpdateScript [
   &EmitterState(Life Cycle Mode=::NewEnumerator1, Loop Duration=1, UseLoopDelay=false)
   &SpawnRate(Spawn Probability=$User.LandBurst, SpawnRate=100, Use Spawn Probability=true)
  ]
  particleSpawnScript [
   /SetVariables(
    Particles.BoneFlow Height=1
    Particles.Boneflow Radius:NiagaraFloat=&Lerp_Float(A=8, B=5, Alpha=$Particles.BoneFlow Height)
    Particles.RandomRangeAngle:NiagaraFloat=&RandomRangeFloat(Minimum=-0.3, Maximum=0.3)
   ) {isSetParameters=true, module=SetVariables_#4}
   &FN_Particle_Partition_v5(Pattern:ENiagaraPartitionWaveForm=::NewEnumerator1)
   &InitializeParticle(
    Lifetime=&RandomRangeFloat(Minimum=1.5, Maximum=2)
    Module.Write Color=true
    Module.Write Lifetime=true
    Module.Write Mass=true
    Module.Write Material Random=true
    Module.Write Position=false
    Module.Write Scale=false
    Module.Write SpriteRotation=true
    Module.Write SpriteSize=true
    Position=&SimulationPosition()
    Sprite Rotation=&RandomRangeFloat(Minimum=0, Maximum=1)
    Sprite Size=&Vector2DFromFloat(Value=&RandomRangeFloat(Minimum=2, Maximum=5))
   )
   &Color(Color=<rgba 0.4 0.59 1 1>, Scale Alpha=$Particles.MaterialRandom)
   &FN_BoneFlow(
    Angle:NiagaraFloat=&Lerp_Float(
     A=$Particles.RandomRangeAngle
     B=&Add_Float(A=$Particles.RandomRangeAngle, B=0.5)
     Alpha=&ModuloFloat(A=&MakeFloatFromInt(Int=$Particles.Partition.ID), B=2)
    )
    Height:NiagaraFloat=&Max_Float(A:NiagaraFloat=$Particles.MaterialRandom, B:NiagaraFloat=0.001)
    Interpolate Mesh Sampling=false
    Radius:NiagaraFloat=$Particles.Boneflow Radius
    SkeletalMesh:NiagaraDataInterfaceSkeletalMesh=$User.SkeletalMesh
    Write Position Attribute=true
   )
   &InheritVelocity(Module.Limit Speed=true, Velocity Scale=&VectorFromFloat(Value=5))
   &AddVelocity(Velocity=&Multiply_VectorByFloat(Vector=$Output.FN_BoneFlow.SplineTangent, Float=-5))
   &AddVelocity(Velocity=&Multiply_VectorByFloat(Vector=$Output.FN_BoneFlow.RotatedBiNormal, Float=5)) {enabled=false, module=AddVelocity001}
  ]
  particleUpdateScript [
   /SetVariables(Particles.Boneflow Radius:NiagaraFloat=&Multiply_Float(
     A=$Particles.Boneflow Radius
     B=&FloatFromCurve(FloatCurve=DataInterface() {propertyValues {Curve {keys [<at 0 1>,<at 1 15>]}}})
    )) {enabled=false, isSetParameters=true, module=SetVariables_#5}
   &FN_BoneFlow(
    Angle:NiagaraFloat=&Lerp_Float(
     A=$Particles.RandomRangeAngle
     B=&Add_Float(A=$Particles.RandomRangeAngle, B=0.5)
     Alpha=&ModuloFloat(A=&MakeFloatFromInt(Int=$Particles.Partition.ID), B=2)
    )
    Height:NiagaraFloat=&Max_Float(A:NiagaraFloat=$Particles.MaterialRandom, B:NiagaraFloat=0.001)
    Interpolate Mesh Sampling=false
    Radius:NiagaraFloat=&Multiply_Float(
     A=$Particles.Boneflow Radius
     B=&FloatFromCurve(FloatCurve=DataInterface() {propertyValues {Curve {keys [<at 0 1>,<at 1 2>]}}}, Scale Curve=5)
    )
    SkeletalMesh:NiagaraDataInterfaceSkeletalMesh=$User.SkeletalMesh
    Write Position Attribute=true
   ) {enabled=false, module=FN_BoneFlow001}
   &AccelerationForce(Acceleration=<0 0 -50>)
   &Drag(
    Drag=&RandomRangeFloat(Minimum=2, Maximum=5)
    Rotational Drag=20
    Use Linear Drag=true
    Use Rotational Drag=false
   )
   &SolveForcesAndVelocity(Clamp Velocity=false, Limit Acceleration=false)
   &ParticleState()
   &SpriteRotationRate(Rotation Rate=&FloatFromCurve(
     FloatCurve=DataInterface() {propertyValues {Curve {keys [<at 0 0 RCIM_Cubic>,<at 0.3 1 RCIM_Cubic>,<at 1 0>]}}}
     Scale Curve=&RandomRangeFloat(Minimum=-100, Maximum=100)
    ))
   &ScaleColor(
    Scale Alpha=&FloatFromCurve(FloatCurve=DataInterface() {propertyValues {Curve {keys [<at 0.6 1>,<at 1 0>]}}})
    ScaleA=true
    ScaleRGB=true
   )
   &ScaleSpriteSize(Scale Factor=&Vector2DFromFloat(Value=&FloatFromCurve(FloatCurve=DataInterface() {propertyValues {Curve {keys [
          <at 0 0 RCIM_Cubic>
          {arriveTangent=2, interpMode=RCIM_Cubic, leaveTangent=2, time=0.1619, value=1}
          {arriveTangent=-1.1932, interpMode=RCIM_Cubic, leaveTangent=-1.1932, time=0.5, value=1}
          <at 1 0 RCIM_Cubic>
         ]}}})))
   &ScaleColor(
    Scale Alpha=&FloatFromCurve(
     FloatCurve=$System.ScaleDistanceFadeCurve
     CurveIndex=&Multiply_Float(
      A=&VectorLength(Vector=&Subtract_Vector(
        A=Linked() {linkedVariable {name=Particles.Position, type=NiagaraPosition}}
        B=&Add_Vector(A=Linked() {linkedVariable {name=Engine.Owner.Position, type=NiagaraPosition}}, B=$System.ScaleDistanceOffset)
       ))
      B=$System.ScaleDistanceBias
     )
    )
    ScaleA=true
    ScaleRGB=true
   ) {module=ScaleColor001}
   /SetVariables(Particles.VisibilityTag=&MakeCustomIntFromBool(Bool=$User.bisHandHidden, True INT=1, False INT=0)) {isSetParameters=true, module=SetVariables_#11}
  ]
  rendererClasses [&NiagaraSpriteRendererProperties]
  renderers [{rendererClass=&NiagaraSpriteRendererProperties}]
 }
 Landing_ColdSmoke {
  emitterUpdateScript [
   &EmitterState(Life Cycle Mode=::NewEnumerator1, Loop Duration=1, UseLoopDelay=false)
   &SpawnRate(Spawn Probability=$User.LandBurst, SpawnRate=100, Use Spawn Probability=true)
  ]
  particleSpawnScript [
   &InitializeParticle(
    Color=&MakeLinearColorFromVectorAndFloat("Vector (RGB)"=<1 1 1>, "Float (Alpha)"=&RandomRangeFloat(Minimum=0.2, Maximum=0.5))
    Lifetime=&RandomRangeFloat(Minimum=0.7, Maximum=1.4)
    Module.Write Color=true
    Module.Write Lifetime=true
    Module.Write Mass=true
    Module.Write Material Random=true
    Module.Write Position=false
    Module.Write Scale=false
    Module.Write SpriteRotation=true
    Module.Write SpriteSize=true
    Position=&SimulationPosition()
    Sprite Rotation=&RandomRangeFloat(Minimum=0, Maximum=360)
    Sprite Size=&Vector2DFromFloat(Value=&RandomRangeFloat(Minimum=15, Maximum=24))
   )
   &TorusLocation(
    Align Torus With Axis=false
    Handle Radius=0
    Large Radius:NiagaraFloat=&RandomRangeFloat(Minimum=8, Maximum=16)
    MaskSpawnGroup=false
    Module.Override Randomness=false
   )
   &AddVelocityFromPoint(Module.Use Falloff Distance=false, Velocity Strength=&RandomRangeFloat(Minimum=60, Maximum=120))
   &CameraOffset(Camera Offset Amount=65) {enabled=false}
   &AddVelocity(Velocity=&MakeVector(X=0, Y=0, Z=&RandomRangeFloat(Minimum=50, Maximum=150)))
  ]
  particleUpdateScript [
   &ParticleState()
   &AccelerationForce(Acceleration=<0 0 -150>)
   &Drag(Drag=&RandomRangeFloat(Minimum=2, Maximum=5), Use Linear Drag=true, Use Rotational Drag=false)
   &SolveForcesAndVelocity(Clamp Velocity=false, Limit Acceleration=false)
   &SubUVAnimation(
    Sub UV Play Rate:NiagaraFloat=&Multiply_Float(A=0.3, B=$Particles.NormalizedAge)
    SubUV Animation Mode:ENiagara_SubUVLookupMode=::NewEnumerator3
   ) {module=SubUVAnimation001}
   &ScaleColor(
    Scale Alpha=0.05
    Scale RGB=&MakeVectorFromLinearColorRGB(Color=&ColorFromCurve(ColorCurve=DataInterface() {propertyValues {
        AlphaCurve {keys [<at 0 0 RCIM_Cubic>,<at 0.04 1>,<at 0.135 0.62>,<at 1.0115 0>,<at 1.094 0>]}
        BlueCurve {keys [<at 0 1 RCIM_Cubic>,<at 1 1 RCIM_Cubic>]}
        GreenCurve {keys [<at 0 0.715 RCIM_Cubic>,<at 1 0.79 RCIM_Cubic>]}
        RedCurve {keys [<at 0 0.62 RCIM_Cubic>,<at 1 0.7 RCIM_Cubic>]}
       }}))
    ScaleA=true
    ScaleRGB=true
   ) {enabled=false}
   &ScaleColor(
    Scale Mode=::NewEnumerator1
    Scale RGBA=&MakeVector4_FromLinearColor(LinearColor=&ColorFromCurve(ColorCurve=DataInterface() {propertyValues {
        AlphaCurve {keys [<at 0 0 RCIM_Cubic>,<at 0.04 1>,<at 0.135 0.62>,<at 1.0115 0>,<at 1.094 0>]}
        BlueCurve {keys [<at 0 1 RCIM_Cubic>,<at 1 1 RCIM_Cubic>]}
        GreenCurve {keys [<at 0 0.715 RCIM_Cubic>,<at 1 0.79 RCIM_Cubic>]}
        RedCurve {keys [<at 0 0.62 RCIM_Cubic>,<at 1 0.7 RCIM_Cubic>]}
       }}))
    ScaleRGBA=true
   ) {module=ScaleColor001}
   &ScaleSpriteSize(Scale Factor=&Vector2DFromFloat(Value=&FloatFromCurve(
      FloatCurve=DataInterface() {propertyValues {Curve {keys [
          {arriveTangent=3.9723, interpMode=RCIM_Cubic, leaveTangent=3.9723, tangentMode=RCTM_User, time=0, value=0}
          {arriveTangent=1, interpMode=RCIM_Cubic, leaveTangent=1, time=0.209, value=0.58}
          <at 1 1 RCIM_Cubic>
         ]}}}
      Scale Curve=3
     )))
   &ScaleColor(
    Scale Alpha=&FloatFromCurve(
     FloatCurve=$System.ScaleDistanceFadeCurve
     CurveIndex=&Multiply_Float(
      A=&VectorLength(Vector=&Subtract_Vector(
        A=Linked() {linkedVariable {name=Particles.Position, type=NiagaraPosition}}
        B=&Add_Vector(A=Linked() {linkedVariable {name=Engine.Owner.Position, type=NiagaraPosition}}, B=$System.ScaleDistanceOffset)
       ))
      B=$System.ScaleDistanceBias
     )
    )
    ScaleA=true
    ScaleRGB=true
   ) {enabled=false, module=ScaleColor002}
   /SetVariables(Particles.VisibilityTag=&MakeCustomIntFromBool(Bool=$User.bisFootHidden, True INT=1, False INT=0)) {isSetParameters=true, module=SetVariables_#12}
  ]
  rendererClasses [&NiagaraSpriteRendererProperties]
  renderers [{rendererClass=&NiagaraSpriteRendererProperties}]
 }
 Landing_SmokePuffs2 {
  emitterUpdateScript [
   &EmitterState(Life Cycle Mode=::NewEnumerator1, Loop Duration=1, UseLoopDelay=false)
   &SpawnRate(Spawn Probability=$User.LandBurst, SpawnRate=100, Use Spawn Probability=true)
   &SpawnBurst_Instantaneous(Spawn Count=12, Spawn Probability=$User.LandBurst, Use Spawn Probability=true) {enabled=false}
  ]
  particleSpawnScript [
   &InitializeParticle(
    Lifetime=&RandomRangeFloat(Minimum=0.7, Maximum=1.4)
    Module.Write Color=true
    Module.Write Lifetime=true
    Module.Write Mass=true
    Module.Write Material Random=true
    Module.Write Position=false
    Module.Write Scale=false
    Module.Write SpriteRotation=true
    Module.Write SpriteSize=true
    Position=&SimulationPosition()
    Sprite Rotation=&RandomRangeFloat(Minimum=0, Maximum=360)
    Sprite Size=&Vector2DFromFloat(Value=&RandomRangeFloat(Minimum=15, Maximum=35))
   )
   &TorusLocation(
    Align Torus With Axis=false
    Handle Radius=0
    Large Radius:NiagaraFloat=&RandomRangeFloat(Minimum=5, Maximum=8)
    MaskSpawnGroup=false
    Module.Override Randomness=false
   )
   &AddVelocityFromPoint(Module.Use Falloff Distance=false, Velocity Strength=60)
   &CameraOffset(Camera Offset Amount=65) {enabled=false}
   &AddVelocity(Velocity=<0 0 100>)
  ]
  particleUpdateScript [
   &ParticleState()
   &AccelerationForce(Acceleration=<0 0 -150>)
   &Drag(Drag=&RandomRangeFloat(Minimum=2, Maximum=3), Use Linear Drag=true, Use Rotational Drag=false)
   &SolveForcesAndVelocity(Clamp Velocity=false, Limit Acceleration=false)
   /Niagara/Modules/Update/SubUV/SubUVAnimation(
    Number Of Frames=&Add_Float(
     A=&Lerp_Float(A=0, B=30, Alpha=$Particles.NormalizedAge)
     B=&RandomRangeFloat(Minimum=0, Maximum=30)
    )
    SubUV Lookup Index=&Multiply_Float(A=$Particles.NormalizedAge, B=1)
   )
   &ScaleColor(
    Scale Mode=::NewEnumerator1
    Scale RGBA=&MakeVector4_FromLinearColor(LinearColor=&ColorFromCurve(ColorCurve=DataInterface() {propertyValues {
        AlphaCurve {keys [<at 0 0 RCIM_Cubic>,<at 0.04 1>,<at 0.135 0.62>,<at 1.0115 0>,<at 1.094 0>]}
        BlueCurve {keys [<at 0 1 RCIM_Cubic>,<at 1 1 RCIM_Cubic>]}
        GreenCurve {keys [<at 0 0.715 RCIM_Cubic>,<at 1 0.79 RCIM_Cubic>]}
        RedCurve {keys [<at 0 0.62 RCIM_Cubic>,<at 1 0.7 RCIM_Cubic>]}
       }}))
    ScaleRGBA=true
   ) {module=ScaleColor001}
   &ScaleSpriteSize(Scale Factor=&Vector2DFromFloat(Value=&FloatFromCurve(
      FloatCurve=DataInterface() {propertyValues {Curve {keys [
          {arriveTangent=3.9723, interpMode=RCIM_Cubic, leaveTangent=3.9723, tangentMode=RCTM_User, time=0, value=0}
          {arriveTangent=1, interpMode=RCIM_Cubic, leaveTangent=1, time=0.209, value=0.58}
          <at 1 1 RCIM_Cubic>
         ]}}}
      Scale Curve=3
     )))
   &ScaleColor(
    Scale Alpha=&FloatFromCurve(
     FloatCurve=$System.ScaleDistanceFadeCurve
     CurveIndex=&Multiply_Float(
      A=&VectorLength(Vector=&Subtract_Vector(
        A=Linked() {linkedVariable {name=Particles.Position, type=NiagaraPosition}}
        B=&Add_Vector(A=Linked() {linkedVariable {name=Engine.Owner.Position, type=NiagaraPosition}}, B=$System.ScaleDistanceOffset)
       ))
      B=$System.ScaleDistanceBias
     )
    )
    ScaleA=true
    ScaleRGB=true
   ) {enabled=false}
   /SetVariables(Particles.VisibilityTag=&MakeCustomIntFromBool(Bool=$User.bisFootHidden, True INT=1, False INT=0)) {isSetParameters=true, module=SetVariables_#12}
  ]
  rendererClasses [&NiagaraSpriteRendererProperties]
  renderers [{rendererClass=&NiagaraSpriteRendererProperties}]
 }
 Landing_SnowFlakes {
  emitterUpdateScript [
   &EmitterState(Life Cycle Mode=::NewEnumerator1, Loop Duration=1, UseLoopDelay=false)
   &SpawnRate(Spawn Probability=$User.LandBurst, SpawnRate=200, Use Spawn Probability=true)
  ]
  particleSpawnScript [
   &InitializeParticle(
    Lifetime=&RandomRangeFloat(Minimum=0.5, Maximum=1.2)
    Module.Write Color=true
    Module.Write Lifetime=true
    Module.Write Mass=true
    Module.Write Material Random=true
    Module.Write Position=false
    Module.Write Scale=false
    Module.Write SpriteRotation=true
    Module.Write SpriteSize=true
    Position=&SimulationPosition()
    Sprite Rotation=&RandomRangeFloat(Minimum=0, Maximum=360)
    Sprite Size=&Vector2DFromFloat(Value=&RandomRangeFloat(Minimum=2, Maximum=5))
   )
   &Color(Color=<rgba 0.4 0.59 1 1>, Scale Alpha=$Particles.MaterialRandom)
   &TorusLocation(
    Align Torus With Axis=false
    Handle Radius=0
    Large Radius:NiagaraFloat=&RandomRangeFloat(Minimum=15, Maximum=18)
    MaskSpawnGroup=false
    Module.Override Randomness=false
   )
   &AddVelocityFromPoint(Module.Use Falloff Distance=false)
   &CameraOffset(Camera Offset Amount=65) {enabled=false}
   &AddVelocity(Velocity=<0 0 100>)
  ]
  particleUpdateScript [
   &ParticleState()
   &AccelerationForce(Acceleration=<0 0 -50>)
   &Drag(Drag=&RandomRangeFloat(Minimum=2, Maximum=3), Use Linear Drag=true, Use Rotational Drag=false)
   &SolveForcesAndVelocity(Clamp Velocity=false, Limit Acceleration=false)
   &ScaleColor(
    Scale RGB=&MakeVectorFromLinearColorRGB(Color=&ColorFromCurve(ColorCurve=DataInterface() {propertyValues {
        AlphaCurve {keys [<at 0 0 RCIM_Cubic>,<at 0.04 1>,<at 0.135 0.62>,<at 1.0115 0>,<at 1.094 0>]}
        BlueCurve {keys [<at 0 1 RCIM_Cubic>,<at 1 1 RCIM_Cubic>]}
        GreenCurve {keys [<at 0 0.8875 RCIM_Cubic>,<at 1 0.79 RCIM_Cubic>]}
        RedCurve {keys [<at 0 0.85 RCIM_Cubic>,<at 1 0.7 RCIM_Cubic>]}
       }}))
    ScaleA=true
    ScaleRGB=true
   )
   &ScaleSpriteSize(Scale Factor=&Vector2DFromFloat(Value=&FloatFromCurve(FloatCurve=DataInterface() {propertyValues {Curve {keys [
          <at 0 0 RCIM_Cubic>
          {arriveTangent=2, interpMode=RCIM_Cubic, leaveTangent=2, time=0.1619, value=1}
          {arriveTangent=-1.1932, interpMode=RCIM_Cubic, leaveTangent=-1.1932, time=0.5, value=1}
          <at 1 0 RCIM_Cubic>
         ]}}})))
   &ScaleColor(
    Scale Alpha=&FloatFromCurve(
     FloatCurve=DataInterface() {propertyValues {Curve {keys [{arriveTangent=-2.3974, interpMode=RCIM_Cubic, leaveTangent=-2.3974, tangentMode=RCTM_User, time=0, value=1}, <at 1 0 RCIM_Cubic>]}}}
     CurveIndex=&Multiply_Float(
      A=&VectorLength(Vector=&Subtract_Vector(
        A=Linked() {linkedVariable {name=Particles.Position, type=NiagaraPosition}}
        B=&Add_Vector(A=Linked() {linkedVariable {name=Engine.Owner.Position, type=NiagaraPosition}}, B=$System.ScaleDistanceOffset)
       ))
      B=0.003
     )
    )
    ScaleA=true
    ScaleRGB=true
   ) {enabled=false, module=ScaleColor002}
   /SetVariables(Particles.VisibilityTag=&MakeCustomIntFromBool(Bool=$User.bisFootHidden, True INT=1, False INT=0)) {isSetParameters=true, module=SetVariables_#12}
  ]
  rendererClasses [&NiagaraSpriteRendererProperties]
  renderers [{rendererClass=&NiagaraSpriteRendererProperties}]
 }
}
summary {
 systemName=NS_Snowman_Idle_Jump
 emitters [
  {emitterName=Landing_ColdSmoke, rendererClasses [&NiagaraSpriteRendererProperties]}
  {emitterName=Idle_Flakes, rendererClasses [&NiagaraSpriteRendererProperties]}
  {emitterName=Landing_SmokePuffs2, rendererClasses [&NiagaraSpriteRendererProperties]}
  {emitterName=Landing_SnowFlakes, rendererClasses [&NiagaraSpriteRendererProperties]}
  {emitterName=JumpSpawnRate001, rendererClasses [&NiagaraSpriteRendererProperties]}
  {emitterName=JumpBurst001, rendererClasses [&NiagaraSpriteRendererProperties]}
  {emitterName=Idle_Puffs, rendererClasses [&NiagaraSpriteRendererProperties]}
  {emitterName=Land_ArmsSnowflakes, rendererClasses [&NiagaraSpriteRendererProperties]}
  {emitterName=Land_ArmsSnowPuffs, rendererClasses [&NiagaraSpriteRendererProperties]}
 ]
 userVariables [
  User.bisFootHidden {type=NiagaraBool}
  User.bisHandHidden {type=NiagaraBool}
  User.BlackSmokeTrailSpawnRate {type=NiagaraFloat, defaultValue=10, description=""}
  User.ColorChange {type=LinearColor, description="", defaultValue {struct=LinearColor, value=<rgba 1 1 1 1>}}
  User.EyeGlowMultiplier {type=NiagaraFloat, defaultValue=0, description=""}
  User.EyeGlowStrength {type=NiagaraFloat, defaultValue=0, description=""}
  User.JumpBurst {type=NiagaraFloat, defaultValue=0, description=""}
  User.JumpTimeRange {type=NiagaraFloat, defaultValue=0, description=""}
  User.LandBurst {type=NiagaraFloat, defaultValue=0, description=""}
  User.LightStrength {type=NiagaraFloat, defaultValue=0, description=""}
  User.Right2ImpactVelocity {type=Vector3f, description="", defaultValue {struct=Vector3f, value=<0 0 0>}}
  User.SkeletalMesh {
   type=NiagaraDataInterfaceSkeletalMesh
   defaultValue {
    struct=NiagaraExt_VariableValue_DataInterface
    value {dataInterface=&self:NiagaraDataInterfaceSkeletalMesh_1}
   }
  }
 ]
}
